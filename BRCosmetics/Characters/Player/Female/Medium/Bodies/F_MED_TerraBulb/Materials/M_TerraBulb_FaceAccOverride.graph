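# Character material for the TerraBulb female skin's face accessory, overriding base color and emissive between the shared 2023 character PreFX and PostFX functions.
# A TechArt mask darkens base color: G with an inverted-Fresnel multiplier, R (flower) with a scalar, B to black.
# The R region also adds base-color glow and a cyan Fresnel rim, with the color multiplied by itself, onto the parent's emissive.
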Material /BRCosmetics/Characters/Player/Female/Medium/Bodies/F_MED_TerraBulb/Materials/M_TerraBulb_FaceAccOverride
graph {
 properties {
  bAutomaticallySetUsageInEditor=false
  bUseMaterialAttributes=true
  bUsedWithMeshDeformer=true
  bUsedWithMorphTargets=true
  bUsedWithSkeletalMesh=true
  pixelDepthOffsetMode=PDOM_Legacy
  shadingModel=MSM_FromMaterialExpression
  usedShadingModels=Default Lit | Subsurface | Clear Coat | Subsurface Profile | Cloth
 }
 nodes [
  n1 = MaterialFunctionCall() {materialFunction=&MF_CharacterParent_2023_PreFX}
  n2 = GetMaterialAttributes(None=@n1.Result) {attributeGetTypes [#1, #2]}
  n3 = NamedRerouteUsage()
  n4 = Constant() {r=1}
  n5 = Fresnel() {baseReflectFraction=0, exponent=1}
  n6 = Multiply(
   A=Power(Base=OneMinus(None=@n5)) {constExponent=4}
   B=ScalarParameter() {defaultValue=0.3, group=Override, parameterName=OverrideBaseColorMultiDark}
  )
  n7 = TextureSampleParameter2D() {
   group=Override
   parameterName=FaceAccOverride
   samplerType=SAMPLERTYPE_Masks
   texture=/BRCosmetics/Characters/Player/Female/Medium/Bodies/F_MED_TerraBulb/Textures/T_TerraBulbFaceAcc_TechArtMask
  }
  n8 = LinearInterpolate(
   A=LinearInterpolate(
    A=LinearInterpolate(A=@n4, B=@n6, Alpha=@n7.G) {constA=1, constB=0.2}
    B=ScalarParameter() {defaultValue=0.2, group=Override, parameterName=OverrideBaseColorMultiFlower}
    Alpha=@n7.R
   )
   Alpha=@n7.B
  ) {constB=0}
  n9 = NamedRerouteUsage()
  n10 = ScalarParameter() {defaultValue=0.2, group=Override, parameterName=BaseColorEM_Intensity}
  n11 = Add(A=@n2.Emissive Color, B=Multiply(A=Multiply(A=@n9, B=@n10), B=@n7.R))
  n12 = Fresnel() {baseReflectFraction=0, exponent=1}
  n13 = ScalarParameter() {defaultValue=1.5, group=Override, parameterName=FresnelPower}
  n14 = VectorParameter() {group=Override, parameterName=EmissiveFresnel, defaultValue=<rgba 0 0.7985 1 2>}
  n15 = Multiply(A=@n7.R, B=Multiply(A=Power(Base=@n12, Exp=@n13), B=Multiply(A=@n14.RGB, B=@n14.RGB)))
  MaterialFunctionCall(
   Material=SetMaterialAttributes!(MaterialAttributes=@n2.MaterialAttributes, Base Color=Multiply(A=@n3, B=@n8), Emissive Color=Add(A=@n11, B=@n15)) {attributeSetTypes [#1, #2]}
   M=@n1.M
   Pre Skinned Local Position=@n1.M
   Distance Blend=@n1.M
   Fresnel=@n1.M
  ) {materialFunction=&MF_CharacterParent_2023_PostFX}
  NamedRerouteDeclaration(None=@n2.Base Color) {name=BaseColorIN, nodeColor=<rgba 0.7812 1 0 1>}
 ]
}
