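# TerraBulb character material function that adds a see-through, blurred ribcage look to input material attributes via emissive and opacity, behind a static switch.
# Custom HLSL multi-tap blurs the scene diffuse buffer, with fewer taps for NPCs and view-depth-scaled, dithered step size.
# Scene depth, bone/distance tints, Fresnel and a noise-perturbed pre-skinned sphere mask shape the effect.
# Shading-path and platform switches provide cheaper forward/mobile variants; DistMask is also output.
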MaterialFunction /BRCosmetics/Characters/Player/Female/Medium/Bodies/F_MED_TerraBulb/Materials/MF_TerraBulb_TransBlur
graph {nodes [
  Constant() {r=1}
  Constant() {r=1}
  n1 = GetMaterialAttributes(None=NamedRerouteUsage()) {attributeGetTypes [#1, #2]}
  n2 = NamedRerouteUsage()
  n3 = Multiply(A=Add(A=NamedRerouteUsage()) {constB=0.2}, B=NamedRerouteUsage())
  n4 = ShadingPathSwitch(Default=@n2, Deferred=@n2, Forward=@n3, Mobile=@n3)
  n5 = MaterialFunctionCall!("Xb1,PS4"=@n4, "PC-High&&Epic,XSX,PS5"=@n4, "PC-Perf,Android"=@n3, Switch=@n3, "PC-Low(NotPerf)&&Medium"=@n4) {materialFunction=/Game/Effects/Niagara/MasterMaterials/Functions/Common/PlatformSwitch_3D}
  n6 = LinearInterpolate(A=@n5.Result, B=NamedRerouteUsage(), Alpha=NamedRerouteUsage())
  n7 = Multiply(A=Multiply(A=NamedRerouteUsage(), B=ScalarParameter() {defaultValue=6, group=Blur, parameterName=NoiseStrength}), B=NamedRerouteUsage())
  n8 = NamedRerouteUsage()
  n9 = NamedRerouteUsage()
  n10 = Constant() {r=1}
  n11 = ShadingPathSwitch(Default=@n9, Deferred=@n9, Forward=@n10, Mobile=@n10)
  n12 = Multiply(
   A=Desaturation(None=@n5.Result, Fraction=Constant() {r=1})
   B=Multiply(
    A=Multiply(A=@n7, B=@n8)
    B=MaterialFunctionCall("Xb1,PS4"=@n11, "PC-High&&Epic,XSX,PS5"=@n11, "PC-Perf,Android"=@n10, Switch=@n10, "PC-Low(NotPerf)&&Medium"=@n11) {materialFunction=&PlatformSwitch_3D}.Result
   )
  )
  n13 = Add(A=NamedRerouteUsage(), B=NamedRerouteUsage())
  n14 = SetMaterialAttributes!(
   MaterialAttributes=@n1.MaterialAttributes
   EmissiveColor=Add(A=@n1.EmissiveColor, B=Add(A=@n6, B=@n12))
   Opacity=LinearInterpolate(A=ScalarParameter() {defaultValue=0.2, group=OpacityCtrl, parameterName=MinOpacity}, Alpha=Saturate(None=@n13)) {constA=0.5}
  ) {attributeSetTypes [#2, #3]}
  n15 = NamedRerouteUsage()
  n16 = StaticBoolParameter() {defaultValue=true, group=PiteousKicksPineChest, parameterName="UseTransparentRibcage (Requires Translucent Blend Mode)", sortPriority=1}
  FunctionOutput(None=NamedRerouteDeclaration(None=StaticSwitch(True=@n14, False=@n15, Value=@n16)) {name=MaterialOut, nodeColor=<rgba 0.7812 1 0 1>})
  FunctionOutput(None=NamedRerouteUsage()) {outputName=DistMask, sortPriority=1}
  n17 = TransformPosition(None=WorldPosition().XYZ) {transformSourceType=TRANSFORMPOSSOURCE_World, transformType=TRANSFORMPOSSOURCE_View}
  Length(None=@n17) {is=class MaterialExpressionLength}
  n18 = VectorParameter() {group=Blur, parameterName=BoneColor, defaultValue=<rgba 0.0371 0.0573 0.0401 1>}
  MaterialExpressionNamedRerouteUsage_2 = NamedRerouteUsage()
  MaterialExpressionQualitySwitch_0 = QualitySwitch(Default=Constant() {r=1}, Epic=Constant() {r=1})
  n19 = Custom!(stepSize=@MaterialExpressionNamedRerouteUsage_2, iterations=@MaterialExpressionQualitySwitch_0) {
   code="float dist = stepSize / iterations;\r\nint BufferIndex = 2;\r\nfloat2 offsets[4] = { float2(1,0), float2(-1,0), float2(0,1), float2(0,-1)};\r\nfloat2 UV = GetDefaultSceneTextureUV(Parameters, int(BufferIndex));\r\nfloat2 NewUV = UV;\r\nfloat3 CurColor=0;\r\n\r\nfor (int i = 0; i < (int)iterations; i++)\r\n{\r\n\tfor (int j = 0; j < 4; j++)\r\n\t{\r\n\t\tNewUV = UV + offsets[j] * (i+1) * dist;\r\n\t\tCurColor += SceneTextureLookupCommon(Parameters,NewUV, BufferIndex, false);\r\n\t}\r\n}\r\n\r\nCurColor /= (iterations * 4.0);\r\nreturn CurColor;\r\n"
   inputs [
    {
     inputName=stepSize
     input {expression=&self:MaterialExpressionNamedRerouteUsage_2}
    }
    {
     inputName=iterations
     input {expression=&self:MaterialExpressionQualitySwitch_0}
    }
   ]
  }
  n20 = Custom(stepSize=@MaterialExpressionNamedRerouteUsage_2, iterations=@MaterialExpressionQualitySwitch_0) {
   code="float dist = stepSize / iterations;\r\nint BufferIndex = 2;\r\nfloat2 offsets[8] = { float2(1,0), float2(-1,0), float2(0,1), float2(0,-1),float2(.5,.5), float2(-.5,-.5), float2(-.5,.5), float2(.5,-.5)};\r\nfloat2 UV = GetDefaultSceneTextureUV(Parameters, int(BufferIndex));\r\nfloat2 NewUV = UV;\r\nfloat3 CurColor=0;\r\n\r\nfor (int i = 0; i < (int)iterations; i++)\r\n{\r\n\tfor (int j = 0; j < 8; j++)\r\n\t{\r\n\t\tNewUV = UV + offsets[j] * (i+1) * dist;\r\n\t\tCurColor += SceneTextureLookupCommon(Parameters,NewUV, BufferIndex, false);\r\n\t}\r\n}\r\n\r\nCurColor /= (iterations * 8.0);\r\nreturn CurColor;\r\n"
   inputs [
    {
     inputName=stepSize
     input {expression=&self:MaterialExpressionNamedRerouteUsage_2}
    }
    {
     inputName=iterations
     input {expression=&self:MaterialExpressionQualitySwitch_0}
    }
   ]
  }
  n21 = StaticSwitchParameter(True=@n19, False=@n20) {group=PiteousKicksPineChest_NPC, parameterName=IsNPC}
  n22 = Constant() {r=1}
  n23 = OneMinus(None=Saturate(None=Desaturation(None=@n21, Fraction=@n22)))
  n24 = Add(
   A=Multiply(A=Multiply(A=@n18.RGB, B=@n18.RGB), B=NamedRerouteDeclaration(None=@n23) {name=BoneArea, nodeColor=<rgba 1 0 0.2578 1>})
   B=Multiply(A=ScalarParameter() {defaultValue=1, group=Blur, parameterName=MeatIntensity}, B=@n21)
  )
  n25 = Custom!() {
   code="float3 CurColor=0;\r\nint BufferIndex = 2;\r\n\r\nfloat2 UV = GetDefaultSceneTextureUV(Parameters, int(BufferIndex));\r\nCurColor = SceneTextureLookupCommon(Parameters, UV, BufferIndex, false);\r\n\r\n\r\nreturn CurColor;\r\n"
   inputs []
  }
  n26 = SceneTexture() {sceneTextureId=PPI_SceneDepth}
  n27 = SceneDepth()
  n28 = Subtract(
   A=ComponentMask(None=ShadingPathSwitch(Default=@n26.Color, Deferred=@n26.Color, Forward=@n27.R, Mobile=@n27.R)) {r=true}
   B=NamedRerouteUsage()
  )
  n29 = ScalarParameter() {defaultValue=5, group=Blur, parameterName=DistAlpha}
  n30 = Power(Base=Saturate(None=Divide(A=@n28, B=@n29)), Exp=ScalarParameter() {defaultValue=2, group=Blur, parameterName=ScenDepthPower02})
  LinearInterpolate(A=@n24, B=Multiply(A=@n25, B=ScalarParameter() {defaultValue=0.6, group=Blur, parameterName=DiffuseMulti}), Alpha=OneMinus(None=@n30))
  n31 = VectorParameter() {group=Banana, parameterName=FresnelColor02, defaultValue=<rgba 0.0045 1 0.7005 1>}
  n32 = Multiply(A=@n31.RGB, B=@n31.RGB)
  LinearInterpolate(
   A=Multiply(A=@n32, B=ScalarParameter() {defaultValue=0.1, group=Banana, parameterName=EmissiveColorOverride})
   B=@n32
   Alpha=Fresnel() {baseReflectFraction=0, exponent=1}
  ) {constA=1}
  n33 = ComponentMask(None=@n17) {b=true}
  n34 = Subtract(A=@n33, B=ScalarParameter() {group=PiteousKicksPineChest, parameterName=BlurDistanceBias})
  n35 = ScalarParameter() {defaultValue=600, group=PiteousKicksPineChest, parameterName=BlurDistanceFade}
  n36 = LinearInterpolate(
   A=QualitySwitch(
    Default=ScalarParameter() {defaultValue=0.005, group=PiteousKicksPineChest, parameterName=BlurNearDefault}
    Epic=ScalarParameter() {defaultValue=0.005, group=PiteousKicksPineChest, parameterName=BlurNearEpic}
   )
   B=ScalarParameter() {defaultValue=0.0005, group=PiteousKicksPineChest, parameterName=BlurFar}
   Alpha=Saturate(None=Divide(A=@n34, B=@n35))
  )
  n37 = Multiply(A=@n36, B=ScalarParameter() {defaultValue=2, group=PiteousKicksPineChest, parameterName=BlurDitherMulitplier}) {constB=2}
  n38 = MaterialFunctionCall() {materialFunction=&DitherTemporalAA}
  NamedRerouteDeclaration(None=LinearInterpolate(
    A=@n36
    B=LinearInterpolate(A=@n36, B=@n37, Alpha=@n38.Result)
    Alpha=ScalarParameter() {defaultValue=1, group=PiteousKicksPineChest, parameterName=BlurUseDither}
   )) {name=BlurDistance, nodeColor=<rgba 0 1 0.9766 1>}
  n39 = ScalarParameter() {defaultValue=50, group=Blur, parameterName=DistFade}
  n40 = ScalarParameter() {defaultValue=10, group=Blur, parameterName=DistFade01}
  n41 = TextureSampleParameter2D() {group=OpacityCtrl, parameterName=DistMask, samplerType=SAMPLERTYPE_Masks, texture=&T_BlackMask}
  n42 = Saturate(None=Divide(A=@n28, B=LinearInterpolate(A=@n39, B=@n40, Alpha=@n41.R)) {constB=20})
  n43 = ScalarParameter() {defaultValue=0.5, group=Blur, parameterName=ScenDepthPower}
  n44 = Saturate(None=Multiply(A=Power(Base=@n42, Exp=@n43), B=ScalarParameter() {defaultValue=0.5, group=Blur, parameterName=ScenDepthIntensity}))
  NamedRerouteDeclaration(None=@n44) {name=SceneDepthAlpha, nodeColor=<rgba 0 0.9297 1 1>}
  n45 = NamedRerouteUsage()
  n46 = ScalarParameter() {defaultValue=10, parameterName=NoiseStrength}
  n47 = SphereMask(
   A=VectorParameter() {group=PiteousKicksPineChest, parameterName=RibCageSphereMaskOffset, defaultValue=<rgba 0 20 127 1>}.RGB
   B=VertexInterpolator(VS=PreSkinnedPosition()).PS
   Radius=Add(A=ScalarParameter() {defaultValue=14, group=PiteousKicksPineChest, parameterName=RibcageSphereMaskRadius}, B=Multiply(A=@n45, B=@n46))
   Hardness=ScalarParameter() {defaultValue=0.7, group=PiteousKicksPineChest, parameterName=RibcageSphereMaskHardnessHardness}
  )
  NamedRerouteDeclaration(None=Multiply(A=@n47, B=ScalarParameter() {defaultValue=0.6, group=PiteousKicksPineChest, parameterName=RibcageMaskStrengthEmissive})) {name=RibcageSphereMaskEmissive, nodeColor=<rgba 1 0.3984 0 1>}
  NamedRerouteDeclaration(None=@n1.Normal) {name=Normal_IN, nodeColor=<rgba 0.9688 0 1 1>}
  n48 = Fresnel(
   ExponentIn=ScalarParameter() {defaultValue=2, group=PiteousKicksPineChest_Fresnel, parameterName=FresnelPow}
   Normal=Transform(None=NamedRerouteUsage())
  ) {baseReflectFraction=0, exponent=1}
  n49 = ScalarParameter() {defaultValue=1, group=PiteousKicksPineChest_Fresnel, parameterName=FresnelAlpha}
  NamedRerouteDeclaration(None=Saturate(None=Multiply(A=@n48, B=@n49))) {name=Fresnel, nodeColor=<rgba 0.9922 1 0 1>}
  n50 = VectorParameter() {group=PiteousKicksPineChest_Fresnel, parameterName=FresnelColor, defaultValue=<rgba 0.041 0.1965 0.2812 1>}
  n51 = Multiply(A=@n50.RGB, B=@n50.RGB)
  n52 = Multiply(A=@n51) {constB=2}
  NamedRerouteDeclaration(None=ShadingPathSwitch(Default=@n51, Deferred=@n51, Forward=@n52, Mobile=@n52)) {name=FresnelColor, nodeColor=<rgba 0 1 0.9063 1>}
  NamedRerouteDeclaration(None=@n28) {name=CameraFade, nodeColor=<rgba 0.9453 1 0 1>}
  NamedRerouteDeclaration(None=@n30) {name=DepthOpacity, nodeColor=<rgba 0 0.2734 1 1>}
  NamedRerouteDeclaration(None=@n41.R) {name=DistMask, nodeColor=<rgba 0 0.9297 1 1>}
  n53 = ScalarParameter() {defaultValue=0.65, group=PiteousKicksPineChest, parameterName=RibcageMaskStrengthMobile}
  NamedRerouteDeclaration(None=OneMinus(None=Multiply(A=@n47, B=@n53))) {name=RibcageSphereMaskMobile, nodeColor=<rgba 1 0.3984 0 1>}
  NamedRerouteDeclaration(None=FunctionInput() {inputName=Material, inputType=FunctionInput_MaterialAttributes}) {name=IN_Material, nodeColor=<rgba 1 0 0.3281 1>}
  NamedRerouteDeclaration(None=FunctionInput() {inputName=Noise, inputType=FunctionInput_Scalar, sortPriority=1}) {name=Noise, nodeColor=<rgba 0 0.2031 1 1>}
  n54 = VectorParameter() {group=Blur, parameterName=DistColor, defaultValue=<rgba 0.0096 0.4031 0 1>}
  n55 = Multiply(A=@n54.RGB, B=@n54.RGB)
  NamedRerouteDeclaration(None=LinearInterpolate(A=@n24, B=@n55, Alpha=@n44)) {name=RibSample, nodeColor=<rgba 0 1 0.2734 1>}
  NamedRerouteDeclaration(None=@n33) {name=CameraDist, nodeColor=<rgba 0 1 0.5781 1>}
  NamedRerouteDeclaration(None=@n55) {name=DistColor, nodeColor=<rgba 0.7813 0 1 1>}
  NamedRerouteDeclaration(None=Multiply(A=@n47, B=ScalarParameter() {defaultValue=1, group=PiteousKicksPineChest, parameterName=RibcageMaskStrengthDiffuse})) {name=RibcageSphereMaskDiffuse, nodeColor=<rgba 1 0.3984 0 1>}
  OneMinus(None=Saturate(None=NamedRerouteUsage()))
  n56 = Constant() {r=1000}
  n57 = Add(A=Divide(A=PixelDepth().R, B=@n56), B=ScalarParameter() {defaultValue=0.001, parameterName=bias})
  n58 = Divide(A=ComponentMask(None=SceneTexture() {sceneTextureId=PPI_CustomDepth}.Color) {r=true}, B=@n56)
  Saturate(None=Sign(None=Subtract(A=@n57, B=@n58)))
 ]}
