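# Translucent unlit glass material for the TerraBulb female character body, faking all lighting in emissive.
# Shared character PreFX, NanaEternal noise and TerraBulb TransBlur functions supply base attributes; the rebuilt emissive then passes into PostFX.
# Glass shading combines Fresnel and normal-distance tints, skylight reflection sampling, and hand-built anisotropic GGX-style specular terms using a Custom-node directional light vector.
# A tiled raindrop mask derives droplet normals for tinted rim and specular glints.
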Material /BRCosmetics/Characters/Player/Female/Medium/Bodies/F_MED_TerraBulb/Materials/M_TerraBulb_Glass
graph {
 properties {
  bUseMaterialAttributes=true
  bUsedWithMeshDeformer=true
  bUsedWithMorphTargets=true
  bUsedWithSkeletalMesh=true
  blendMode=BLEND_Translucent
  pixelDepthOffsetMode=PDOM_Legacy
  shadingModel=MSM_Unlit
  usedShadingModels=Unlit
 }
 nodes [
  n1 = TextureSampleParameter2D(UVs=Multiply(A=TextureCoordinate(), B=ScalarParameter() {defaultValue=10, group=GlassOverride, parameterName=DropsTiling})) {
   group=GlassOverride
   parameterName=Drops
   samplerType=SAMPLERTYPE_LinearColor
   texture=/BRCosmetics/Characters/Player/Female/Medium/Bodies/F_MED_TerraBulb/Textures/T_RainDrop_Mask
  }
  n2 = ComponentMask(None=@n1.RGB) {g=true, r=true}
  n3 = Normalize(VectorInput=DeriveNormalZ(InXY=Subtract(A=@n2) {constB=0.5}))
  n4 = Constant3Vector() {constant=<rgba 0 0 1 0>}
  n5 = Transform(None=@n3)
  n6 = DotProduct(A=@n5, B=NamedRerouteUsage())
  n7 = Multiply(A=Power(Base=Distance(A=@n3, B=@n4)), B=LinearInterpolate(Alpha=Saturate(None=@n6)) {constA=0.3})
  n8 = Saturate(None=DotProduct(A=ReflectionVectorWS(CustomWorldNormal=@n5), B=NamedRerouteUsage()))
  n9 = LinearInterpolate(Alpha=@n8) {constA=0.8, constB=2}
  n10 = VectorParameter() {group=GlassOverride, parameterName=RimRainColor, defaultValue=<rgba 0.4847 1 0.9291 10>}
  n11 = VectorParameter() {group=GlassOverride, parameterName=RimRainColor_Spec, defaultValue=<rgba 0.4847 1 0.9291 10>}
  n12 = NamedRerouteUsage()
  n13 = Power(Base=@n8) {constExponent=10}
  n14 = Add(
   A=Multiply(A=Multiply(A=@n7, B=@n9), B=Multiply(A=@n10.RGB, B=@n10.RGB))
   B=Multiply(A=Multiply(A=@n11.RGB, B=@n11.RGB), B=Multiply(A=@n12, B=@n13))
  )
  n15 = Desaturation(None=NamedRerouteUsage(), Fraction=Constant() {r=1})
  Add(A=@n14, B=Multiply(A=SmoothStep(Value=@n15) {constMin=0.2}, B=NamedRerouteUsage()))
  CameraVectorWS()
  Constant() {r=1}
  n16 = NamedRerouteUsage()
  n17 = NamedRerouteUsage()
  n18 = ReflectionVectorWS(CustomWorldNormal=@n17)
  DotProduct(A=@n16, B=@n18)
  n19 = MaterialFunctionCall() {materialFunction=&MF_CharacterParent_2023_PreFX}
  n20 = MaterialFunctionCall!(MaterialAttributes=@n19.Result, Fresnel=@n19.Result) {materialFunction=/BRCosmetics/Characters/Player/Male/Medium/Bodies/M_MED_NanaEternal/Materials/MF_NanaEternal_Noise}
  n21 = MaterialFunctionCall!(Material=@n20.Result, Noise=@n20.Result) {materialFunction=&/MF_TerraBulb_TransBlur}
  n22 = GetMaterialAttributes(None=@n21.Result) {attributeGetTypes [#1, #2]}
  MaterialFunctionCall(
   Material=SetMaterialAttributes(MaterialAttributes=@n22.MaterialAttributes, Emissive Color=NamedRerouteUsage()) {attributeSetTypes [#2]}
   M=@n19.M
   Pre Skinned Local Position=@n19.M
   Distance Blend=@n19.M
   Fresnel=@n19.M
  ) {materialFunction=&MF_CharacterParent_2023_PostFX}
  NamedRerouteDeclaration(None=@n22.Emissive Color) {name=EM_IN, nodeColor=<rgba 0 1 0.9297 1>}
  NamedRerouteDeclaration(None=@n22.Normal) {name=Normal_IN, nodeColor=<rgba 1 0.6094 0 1>}
  n23 = NamedRerouteUsage()
  n24 = NamedRerouteUsage()
  NamedRerouteDeclaration(None=Normalize(VectorInput=Add(A=@n23, B=@n24))) {name=H, nodeColor=<rgba 1 0.5156 0 1>}
  n25 = NamedRerouteUsage()
  n26 = NamedRerouteUsage()
  NamedRerouteDeclaration(None=Saturate(None=DotProduct(A=@n25, B=@n26))) {name=NdotH, nodeColor=<rgba 0.1953 0 1 1>}
  n27 = NamedRerouteUsage()
  n28 = NamedRerouteUsage()
  NamedRerouteDeclaration(None=Saturate(None=DotProduct(A=@n27, B=@n28))) {name=VdotH, nodeColor=<rgba 0.125 1 0 1>}
  NamedRerouteDeclaration(None=DotProduct(A=NamedRerouteUsage(), B=NamedRerouteUsage())) {name=TdotH, nodeColor=<rgba 1 0.3516 0 1>}
  NamedRerouteDeclaration(None=DotProduct(A=NamedRerouteUsage(), B=NamedRerouteUsage())) {name=BdotH, nodeColor=<rgba 0.0312 0 1 1>}
  n29 = CameraVectorWS()
  NamedRerouteDeclaration(None=@n29) {name=V, nodeColor=<rgba 0.2891 1 0 1>}
  NamedRerouteDeclaration(None=DotProduct(A=NamedRerouteUsage(), B=NamedRerouteUsage())) {name=TdotV, nodeColor=<rgba 1 0 0.6094 1>}
  NamedRerouteDeclaration(None=DotProduct(A=NamedRerouteUsage(), B=NamedRerouteUsage())) {name=BdotV, nodeColor=<rgba 0 0.9297 1 1>}
  n30 = NamedRerouteUsage()
  NamedRerouteDeclaration(None=DotProduct(A=NamedRerouteUsage(), B=@n30)) {name=TdotL, nodeColor=<rgba 1 0 0.6094 1>}
  n31 = VectorParameter() {group=GlassOverride, parameterName=FresnelColor_Glass, defaultValue=<rgba 0.2135 0.9586 1 1>}
  n32 = NamedRerouteUsage()
  n33 = Transform(None=@n32)
  n34 = Power(
   Base=Fresnel(Normal=@n33) {baseReflectFraction=0, exponent=1}
   Exp=ScalarParameter() {defaultValue=4, group=GlassOverride, parameterName=GlassFresnelPOW}
  )
  n35 = Saturate(None=Distance(A=@n32, B=Constant3Vector() {constant=<rgba 0 0 1 0>}))
  n36 = ScalarParameter() {defaultValue=4, group=GlassOverride, parameterName=DistancePower}
  n37 = VectorParameter() {group=GlassOverride, parameterName=DistanceColor, defaultValue=<rgba 0.033 0.9755 1 4>}
  n38 = Add(
   A=LinearInterpolate(A=NamedRerouteUsage(), B=Multiply(A=@n31.RGB, B=@n31.RGB), Alpha=Saturate(None=@n34))
   B=Multiply(A=Power(Base=@n35, Exp=@n36), B=Multiply(A=@n37.RGB, B=@n37.RGB))
  )
  n39 = NamedRerouteUsage()
  n40 = VectorParameter() {parameterName=GlassDiffuse, defaultValue=<rgba 0.0955 0.9323 0.8529 1>}
  n41 = Multiply(A=@n40.RGB, B=@n40.RGB)
  n42 = Add(A=LinearInterpolate(A=@n38, B=Multiply(A=@n38) {constB=0.5}, Alpha=@n39), B=Multiply(A=LinearInterpolate(A=@n38, B=@n41), B=@n39))
  n43 = Multiply(A=NamedRerouteUsage()) {constB=0.2}
  NamedRerouteDeclaration(None=Add(A=Add(A=@n42, B=@n43), B=NamedRerouteUsage())) {name=EM_OUT, nodeColor=<rgba 1 0 0.3516 1>}
  NamedRerouteDeclaration(None=DotProduct(A=NamedRerouteUsage(), B=@n30)) {name=BdotL, nodeColor=<rgba 0 0.9297 1 1>}
  n44 = ScalarParameter() {defaultValue=-0.2, group=GlassOverride, parameterName=Anisotropy}
  n45 = OneMinus(None=NamedRerouteUsage())
  NamedRerouteDeclaration(None=Abs(None=Multiply(A=@n44, B=@n45))) {name=a, nodeColor=<rgba 0 1 0.1094 1>}
  n46 = ScalarParameter() {defaultValue=0.1, group=GlassOverride, parameterName=Roughness}
  n47 = NamedRerouteDeclaration(None=Max(B=Multiply(A=@n46, B=@n46)) {constA=0.001}) {name=Alpha, nodeColor=<rgba 1 0 0.2109 1>}
  n48 = Constant() {r=1}
  n49 = Multiply(A=Constant() {r=0.9}, B=NamedRerouteUsage())
  n50 = NamedRerouteDeclaration(None=Max(A=Constant() {r=0.001}, B=Subtract(A=@n48, B=@n49))) {name=aspect, nodeColor=<rgba 0 0.5312 1 1>}
  n51 = Divide(A=@n47, B=@n50)
  n52 = Multiply(A=@n47, B=@n50)
  NamedRerouteDeclaration(None=If(A=NamedRerouteUsage(), "A > B"=@n51, "A == B"=@n51, "A < B"=@n52)) {name=ax, nodeColor=<rgba 0.8516 1 0 1>}
  NamedRerouteDeclaration(None=If(A=NamedRerouteUsage(), "A > B"=@n52, "A == B"=@n52, "A < B"=@n51)) {name=ay, nodeColor=<rgba 0.8281 0 1 1>}
  NamedRerouteDeclaration(None=@n44) {name=Aniso, nodeColor=<rgba 0 1 0.5078 1>}
  n53 = NamedRerouteUsage()
  n54 = NamedRerouteUsage()
  n55 = Divide(A=Multiply(A=@n53, B=@n53), B=Multiply(A=@n54, B=@n54))
  n56 = NamedRerouteUsage()
  n57 = NamedRerouteUsage()
  n58 = Divide(A=Multiply(A=@n56, B=@n56), B=Multiply(A=@n57, B=@n57))
  n59 = NamedRerouteUsage()
  NamedRerouteDeclaration(None=Add(A=Add(A=@n55, B=@n58), B=Multiply(A=@n59, B=@n59))) {name=D, nodeColor=<rgba 1 0.1875 0 1>}
  n60 = NamedRerouteUsage()
  n61 = Multiply(A=@n60, B=@n60)
  n62 = Multiply(
   A=Multiply(A=MaterialFunctionCall() {materialFunction=&Pi}.Result, B=NamedRerouteUsage())
   B=NamedRerouteUsage()
  )
  NamedRerouteDeclaration(None=Divide(B=Multiply(A=@n61, B=@n62))) {name=GGX, nodeColor=<rgba 0 0.1328 1 1>}
  NamedRerouteDeclaration(None=@n21.DistMask) {name=DistMask, nodeColor=<rgba 0 1 0.1094 1>}
  NamedRerouteDeclaration(None=Custom!(None) {code="return View.DirectionalLightDirection;", description=LightDirection, showCode=true}) {name=LightVector, nodeColor=<rgba 0 1 0.9297 1>}
  NamedRerouteDeclaration(None=@n1.B) {name=RainMask, nodeColor=<rgba 1 0 0.9609 1>}
  NamedRerouteDeclaration(None=@n14) {name=RainEM, nodeColor=<rgba 0 1 0.8594 1>}
  NamedRerouteDeclaration(None=@n20.GhostNoise) {name=GhostNoise, nodeColor=<rgba 0 1 0.7891 1>}
  NamedRerouteDeclaration(None=@n41) {name=GlassDiffuse, nodeColor=<rgba 0.7578 0 1 1>}
  NamedRerouteDeclaration(None=@n33) {name=Normal_World, nodeColor=<rgba 1 0.6094 0 1>}
  n63 = Saturate(None=NamedRerouteUsage())
  n64 = ScalarParameter() {defaultValue=2, group=GlassOverride, parameterName=SpecPow}
  NamedRerouteDeclaration(None=SmoothStep(
    Min=ScalarParameter() {group=GlassOverride, parameterName=SpecMin}
    Max=ScalarParameter() {defaultValue=0.02, group=GlassOverride, parameterName=SpecMax}
    Value=Power(Base=@n63, Exp=@n64)
   )) {name=Spec, nodeColor=<rgba 0.2891 0 1 1>}
  n65 = DotProduct(A=@n16, B=@n17)
  n66 = Constant() {r=0.2}
  NamedRerouteDeclaration(None=Multiply(A=Saturate(None=@n65), B=SkyLightEnvMapSample(Direction=@n18, Roughness=@n66))) {name=Reflection, nodeColor=<rgba 0 1 0.3438 1>}
  n67 = NamedRerouteUsage()
  NamedRerouteDeclaration(None=Saturate(None=DotProduct(A=@n67, B=@n29))) {name=NdotV, nodeColor=<rgba 0 0.2031 1 1>}
  n68 = NamedRerouteUsage()
  n69 = NamedRerouteUsage()
  NamedRerouteDeclaration(None=Saturate(None=DotProduct(A=@n68, B=@n69))) {name=NdotL, nodeColor=<rgba 0.5234 1 0 1>}
  n70 = Transform(None=Normalize(VectorInput=VectorParameter() {parameterName=TangentDirTangentSpace, defaultValue=<rgba 1 0 0 1>}.RGB))
  n71 = NamedRerouteUsage()
  n72 = Subtract(A=@n70, B=Multiply(A=@n71, B=DotProduct(A=@n70, B=@n71)))
  n73 = NamedRerouteUsage()
  n74 = Normalize(VectorInput=CrossProduct(A=Normalize(VectorInput=@n72), B=@n73))
  NamedRerouteDeclaration(None=CrossProduct(A=NamedRerouteDeclaration(None=@n74) {name=B, nodeColor=<rgba 0 1 0.8359 1>}, B=@n73)) {name=T, nodeColor=<rgba 1 0 0.8437 1>}
  NamedRerouteUsage()
  NamedRerouteUsage()
  Transform(None=Constant3Vector() {constant=<rgba 1 0 0 0>})
 ]
}
