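# Physics asset for the skydiving variant of the female TerraBulb character's coat, with capsule bodies for coat, bag, lower spine, pelvis and legs.
# Five coat chains (front and side pairs, one back) hang from spine_01 through tiny base bodies, with tight swing limits and locked twist.
# Angularly free constraints cross-link neighboring hem tips and attach the bag to the right thigh.
# A huge root sphere below the character approximates a floor.
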PhysicsAsset /BRCosmetics/Characters/Player/Female/Medium/Bodies/F_MED_TerraBulb/Meshes/F_MED_TerraBulb_Coat_SkyDiving_Physics
bodies {
 calf_l [{shapeType=Capsule, length=47, radius=8.5, shapeName=calf_l_capsule, center=<-25.2791 -2.4051 0>, rotation=<rot -90 180 0>}]
 calf_r [{shapeType=Capsule, length=47, radius=8.5, shapeName=calf_r_capsule, center=<25.2791 2.4051 0>, rotation=<rot -90 0 0>}]
 dyn_bag [{
   shapeType=Capsule
   length=4.5998
   radius=5.5571
   shapeName=dyn_bag_capsule
   center=<3.5067 -0.1714 1.9422>
   rotation=<rot 19.9819 -91.343 -95.5655>
  }]
 dyn_coat_bk_l_01 [{shapeType=Capsule, length=13, radius=7, shapeName=dyn_coat_bk_l_01_capsule, center=<5 0 -6.5>, rotation=<rot 0 0 90>}]
 dyn_coat_bk_l_02 [{shapeType=Capsule, length=15.5, radius=7.5, shapeName=dyn_coat_bk_l_02_capsule, center=<5.5 0 -7>, rotation=<rot 0 0 90>}]
 dyn_coat_bk_l_03 [{shapeType=Capsule, length=20, radius=7.5, shapeName=dyn_coat_bk_l_03_capsule, center=<5.5 0 -4>, rotation=<rot 0 0 90>}]
 dyn_coat_bk_l_base [{shapeType=Capsule, length=0.505, radius=0.505, shapeName=dyn_coat_bk_l_base_capsule}]
 dyn_coat_fr_l_01 [{shapeType=Capsule, length=14.5, radius=5.5, shapeName=dyn_coat_fr_l_01_capsule, center=<10.3 0 -2.5>, rotation=<rot 0 0 -90>}]
 dyn_coat_fr_l_02 [{shapeType=Capsule, length=13, radius=6, shapeName=dyn_coat_fr_l_02_capsule, center=<5.5 0 -3>, rotation=<rot 0 0 -90>}]
 dyn_coat_fr_l_03 [{shapeType=Capsule, length=11.5, radius=9, shapeName=dyn_coat_fr_l_03_capsule, center=<5.5 0 -4.5>, rotation=<rot 0 0 -90>}]
 dyn_coat_fr_l_04 [{shapeType=Capsule, length=17.5, radius=10, shapeName=dyn_coat_fr_l_04_capsule, center=<9 0 -5>, rotation=<rot 0 0 -96>}]
 dyn_coat_fr_l_base [{shapeType=Capsule, length=0.505, radius=0.505, shapeName=dyn_coat_fr_l_base_capsule}]
 dyn_coat_fr_r_01 [{shapeType=Capsule, length=20, radius=5.5, shapeName=dyn_coat_fr_r_01_capsule, center=<10.3 0 -2.5>, rotation=<rot 0 0 91.3848>}]
 dyn_coat_fr_r_02 [{shapeType=Capsule, length=20, radius=6.5, shapeName=dyn_coat_fr_r_02_capsule, center=<5.5 0 -3>, rotation=<rot 0 0.0001 91.6686>}]
 dyn_coat_fr_r_03 [{shapeType=Capsule, length=16, radius=8.5, shapeName=dyn_coat_fr_r_03_capsule, center=<5.5 0 -4.5>, rotation=<rot 0 -0.5618 90.5283>}]
 dyn_coat_fr_r_04 [{
   shapeType=Capsule
   length=17.5
   radius=10
   shapeName=dyn_coat_fr_r_04_capsule
   center=<10.6966 -0.0508 -4.2343>
   rotation=<rot 0 0.0001 86.2065>
  }]
 dyn_coat_fr_r_base [{shapeType=Capsule, length=0.505, radius=0.505, shapeName=dyn_coat_fr_r_base_capsule}]
 dyn_coat_side_l_01 [{shapeType=Capsule, length=14.5, radius=5.5, shapeName=dyn_coat_side_l_01_capsule, center=<8 0 -4>, rotation=<rot 0 0 -90>}]
 dyn_coat_side_l_02 [{shapeType=Capsule, length=14.5, radius=5.5, shapeName=dyn_coat_side_l_02_capsule, center=<5.5 0 -4>, rotation=<rot 0 0 -90>}]
 dyn_coat_side_l_03 [{shapeType=Capsule, length=15.5, radius=7.5, shapeName=dyn_coat_side_l_03_capsule, center=<5.5 0 -4>, rotation=<rot 0 20 -100.3962>}]
 dyn_coat_side_l_base [{shapeType=Capsule, length=0.505, radius=0.505, shapeName=dyn_coat_side_l_base_capsule}]
 dyn_coat_side_r_01 [{shapeType=Capsule, length=14, radius=6, shapeName=dyn_coat_side_r_01_capsule, center=<5.5 0 -4>, rotation=<rot 0 0 90>}]
 dyn_coat_side_r_02 [{shapeType=Capsule, length=22.5, radius=7, shapeName=dyn_coat_side_r_02_capsule, center=<5.5 0 -4>, rotation=<rot 0 0 90>}]
 dyn_coat_side_r_03 [{shapeType=Capsule, length=14, radius=7.5, shapeName=dyn_coat_side_r_03_capsule, center=<5.5 0 -4>, rotation=<rot 0 -20 98.6191>}]
 dyn_coat_side_r_base [{shapeType=Capsule, length=0.505, radius=0.505, shapeName=dyn_coat_side_r_base_capsule}]
 foot_l [{shapeType=Capsule, length=8.9286, radius=10.8294, shapeName=foot_l_capsule, center=<-3.169 6.5374 0>, rotation=<rot 0 0 90>}]
 foot_r [{shapeType=Capsule, length=9.1435, radius=10.8905, shapeName=foot_r_capsule, center=<3.169 -6.5374 0>, rotation=<rot 0 180 90>}]
 pelvis [{shapeType=Capsule, length=5, radius=14, shapeName=pelvis_capsule, center=<-5.2357 0 0>}]
 root [{shapeType=Sphere, radius=250, shapeName=root_sphere, center=<0 0 -245.432>}]
 spine_01 [{shapeType=Capsule, length=5.5, radius=10, shapeName=spine_01_capsule, center=<-2.5215 0 0>}]
 thigh_l [{
   shapeType=Capsule
   length=35.5
   radius=8.5
   shapeName=thigh_l_capsule
   center=<-20.1982 -0.8585 1.0441>
   rotation=<rot -87.563 -0.0007 -179.9997>
  }]
 thigh_r [{
   shapeType=Capsule
   length=35.5
   radius=8.5
   shapeName=thigh_r_capsule
   center=<20.1982 0.8585 -1.0441>
   rotation=<rot -87.563 -0.0007 0.0003>
  }]
}
constraints [
 {bone1Name=dyn_coat_bk_l_02, bone2Name=dyn_coat_bk_l_01, swing1LimitDegrees=5, swing2LimitDegrees=10, twistLimitDegrees=5, twistMotion=Locked}
 {bone1Name=dyn_coat_bk_l_03, bone2Name=dyn_coat_bk_l_02, swing1LimitDegrees=5, swing2LimitDegrees=10, twistLimitDegrees=5, twistMotion=Locked}
 {
  bone1Name=dyn_coat_bk_l_01
  bone2Name=dyn_coat_bk_l_base
  swing1LimitDegrees=5
  swing2LimitDegrees=5
  twistLimitDegrees=5
  twistMotion=Locked
 }
 {
  bone1Name=dyn_coat_fr_l_01
  bone2Name=dyn_coat_fr_l_base
  swing1LimitDegrees=5
  swing2LimitDegrees=10
  twistLimitDegrees=5
  twistMotion=Locked
 }
 {
  bone1Name=dyn_coat_fr_l_02
  bone2Name=dyn_coat_fr_l_01
  swing1LimitDegrees=10
  swing2LimitDegrees=10
  twistLimitDegrees=5
  twistMotion=Locked
 }
 {
  bone1Name=dyn_coat_fr_l_03
  bone2Name=dyn_coat_fr_l_02
  swing1LimitDegrees=10
  swing2LimitDegrees=10
  twistLimitDegrees=5
  twistMotion=Locked
 }
 {
  bone1Name=dyn_coat_fr_l_04
  bone2Name=dyn_coat_fr_l_03
  swing1LimitDegrees=10
  swing2LimitDegrees=10
  twistLimitDegrees=5
  twistMotion=Locked
 }
 {
  bone1Name=dyn_coat_fr_r_01
  bone2Name=dyn_coat_fr_r_base
  swing1LimitDegrees=10
  swing2LimitDegrees=10
  twistLimitDegrees=5
  twistMotion=Locked
 }
 {
  bone1Name=dyn_coat_fr_r_02
  bone2Name=dyn_coat_fr_r_01
  swing1LimitDegrees=10
  swing2LimitDegrees=10
  twistLimitDegrees=5
  twistMotion=Locked
 }
 {
  bone1Name=dyn_coat_fr_r_03
  bone2Name=dyn_coat_fr_r_02
  swing1LimitDegrees=10
  swing2LimitDegrees=10
  twistLimitDegrees=5
  twistMotion=Locked
 }
 {
  bone1Name=dyn_coat_fr_r_04
  bone2Name=dyn_coat_fr_r_03
  swing1LimitDegrees=10
  swing2LimitDegrees=10
  twistLimitDegrees=5
  twistMotion=Locked
 }
 {
  bone1Name=dyn_coat_side_l_01
  bone2Name=dyn_coat_side_l_base
  swing1LimitDegrees=5
  swing2LimitDegrees=10
  twistLimitDegrees=5
  twistMotion=Locked
 }
 {
  bone1Name=dyn_coat_side_l_02
  bone2Name=dyn_coat_side_l_01
  swing1LimitDegrees=10
  swing2LimitDegrees=10
  twistLimitDegrees=5
  twistMotion=Locked
 }
 {
  bone1Name=dyn_coat_side_l_03
  bone2Name=dyn_coat_side_l_02
  swing1LimitDegrees=10
  swing2LimitDegrees=10
  twistLimitDegrees=5
  twistMotion=Locked
 }
 {
  bone1Name=dyn_coat_side_r_01
  bone2Name=dyn_coat_side_r_base
  swing1LimitDegrees=5
  swing2LimitDegrees=10
  twistLimitDegrees=5
  twistMotion=Locked
 }
 {
  bone1Name=dyn_coat_side_r_02
  bone2Name=dyn_coat_side_r_01
  swing1LimitDegrees=5
  swing2LimitDegrees=10
  twistLimitDegrees=5
  twistMotion=Locked
 }
 {
  bone1Name=dyn_coat_side_r_03
  bone2Name=dyn_coat_side_r_02
  swing1LimitDegrees=5
  swing2LimitDegrees=10
  twistLimitDegrees=5
  twistMotion=Locked
 }
 {
  bone1Name=dyn_coat_side_r_base
  bone2Name=spine_01
  swing1LimitDegrees=45
  swing1Motion=Free
  swing2LimitDegrees=45
  swing2Motion=Free
  twistLimitDegrees=45
  twistMotion=Free
 }
 {
  bone1Name=dyn_coat_fr_r_base
  bone2Name=spine_01
  swing1LimitDegrees=45
  swing1Motion=Free
  swing2LimitDegrees=45
  swing2Motion=Free
  twistLimitDegrees=45
  twistMotion=Free
 }
 {
  bone1Name=dyn_coat_bk_l_base
  bone2Name=spine_01
  swing1LimitDegrees=45
  swing1Motion=Free
  swing2LimitDegrees=45
  swing2Motion=Free
  twistLimitDegrees=45
  twistMotion=Free
 }
 {
  bone1Name=dyn_coat_side_l_base
  bone2Name=spine_01
  swing1LimitDegrees=45
  swing1Motion=Free
  swing2LimitDegrees=45
  swing2Motion=Free
  twistLimitDegrees=45
  twistMotion=Free
 }
 {
  bone1Name=dyn_coat_fr_l_base
  bone2Name=spine_01
  swing1LimitDegrees=45
  swing1Motion=Free
  swing2LimitDegrees=45
  swing2Motion=Free
  twistLimitDegrees=45
  twistMotion=Free
 }
 {
  bone1Name=pelvis
  bone2Name=root
  swing1LimitDegrees=45
  swing1Motion=Free
  swing2LimitDegrees=45
  swing2Motion=Free
  twistLimitDegrees=45
  twistMotion=Free
 }
 {
  bone1Name=thigh_l
  bone2Name=pelvis
  swing1LimitDegrees=45
  swing1Motion=Free
  swing2LimitDegrees=45
  swing2Motion=Free
  twistLimitDegrees=45
  twistMotion=Free
 }
 {
  bone1Name=calf_l
  bone2Name=thigh_l
  swing1LimitDegrees=45
  swing1Motion=Free
  swing2LimitDegrees=45
  swing2Motion=Free
  twistLimitDegrees=45
  twistMotion=Free
 }
 {
  bone1Name=thigh_r
  bone2Name=pelvis
  swing1LimitDegrees=45
  swing1Motion=Free
  swing2LimitDegrees=45
  swing2Motion=Free
  twistLimitDegrees=45
  twistMotion=Free
 }
 {
  bone1Name=calf_r
  bone2Name=thigh_r
  swing1LimitDegrees=45
  swing1Motion=Free
  swing2LimitDegrees=45
  swing2Motion=Free
  twistLimitDegrees=45
  twistMotion=Free
 }
 {
  bone1Name=spine_01
  bone2Name=pelvis
  swing1LimitDegrees=45
  swing1Motion=Free
  swing2LimitDegrees=45
  swing2Motion=Free
  twistLimitDegrees=45
  twistMotion=Free
 }
 {
  bone1Name=foot_r
  bone2Name=calf_r
  swing1LimitDegrees=45
  swing1Motion=Free
  swing2LimitDegrees=45
  swing2Motion=Free
  twistLimitDegrees=45
  twistMotion=Free
 }
 {
  bone1Name=foot_l
  bone2Name=calf_l
  swing1LimitDegrees=45
  swing1Motion=Free
  swing2LimitDegrees=45
  swing2Motion=Free
  twistLimitDegrees=45
  twistMotion=Free
 }
 {
  bone1Name=dyn_bag
  bone2Name=thigh_r
  swing1LimitDegrees=45
  swing1Motion=Free
  swing2LimitDegrees=45
  swing2Motion=Free
  twistLimitDegrees=45
  twistMotion=Free
 }
 {
  bone1Name=dyn_coat_bk_l_03
  bone2Name=dyn_coat_side_r_03
  swing1LimitDegrees=45
  swing1Motion=Free
  swing2LimitDegrees=45
  swing2Motion=Free
  twistLimitDegrees=45
  twistMotion=Free
 }
 {
  bone1Name=dyn_coat_bk_l_03
  bone2Name=dyn_coat_side_l_03
  swing1LimitDegrees=45
  swing1Motion=Free
  swing2LimitDegrees=45
  swing2Motion=Free
  twistLimitDegrees=45
  twistMotion=Free
 }
 {
  bone1Name=dyn_coat_fr_l_04
  bone2Name=dyn_coat_side_l_03
  swing1LimitDegrees=45
  swing1Motion=Free
  swing2LimitDegrees=45
  swing2Motion=Free
  twistLimitDegrees=45
  twistMotion=Free
 }
 {
  bone1Name=dyn_coat_fr_r_04
  bone2Name=dyn_coat_side_r_03
  swing1LimitDegrees=45
  swing1Motion=Free
  swing2LimitDegrees=45
  swing2Motion=Free
  twistLimitDegrees=45
  twistMotion=Free
 }
]
