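# Physics asset for the TerraBulb female medium body's hood, defining capsule bodies for its dynamic side and back hood bones plus head, clavicles and upper arms.
# Each hood chain's base joins the head with free angular motion.
# Side tips carry larger capsules with tight swing and twist limits to their bases, while back tips are tiny capsules angularly locked to theirs.
# Upper arms link freely to clavicles, separate from the head-hood tree.
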PhysicsAsset /BRCosmetics/Characters/Player/Female/Medium/Bodies/F_MED_TerraBulb/Meshes/F_MED_TerraBulb_Hood_Physics
bodies {
 clavicle_l [{shapeType=Capsule, length=10, radius=5, shapeName=clavicle_l_capsule, center=<9 0 0>, rotation=<rot 90 0 0>}]
 clavicle_r [{shapeType=Capsule, length=10, radius=5, shapeName=clavicle_r_capsule, center=<-9 0 0>, rotation=<rot 90 0 0>}]
 dyn_hood_bk_l_01 [{shapeType=Capsule, length=0.505, radius=0.505, shapeName=dyn_hood_bk_l_01_capsule, rotation=<rot 0 -90 0>}]
 dyn_hood_bk_l_base [{shapeType=Capsule, length=0.505, radius=0.505, shapeName=dyn_hood_bk_l_base_capsule, rotation=<rot 0 -90 0>}]
 dyn_hood_bk_r_01 [{shapeType=Capsule, length=0.505, radius=0.505, shapeName=dyn_hood_bk_r_01_capsule, rotation=<rot 0 -90 0>}]
 dyn_hood_bk_r_base [{shapeType=Capsule, length=0.505, radius=0.505, shapeName=dyn_hood_bk_r_base_capsule, rotation=<rot 0 -90 0>}]
 dyn_hood_side_l_01 [{shapeType=Capsule, length=5, radius=4, shapeName=dyn_hood_side_l_01_capsule, center=<7 1.1 1>, rotation=<rot 0 0 90>}]
 dyn_hood_side_l_base [{shapeType=Capsule, length=0.505, radius=0.505, shapeName=dyn_hood_side_l_base_capsule}]
 dyn_hood_side_r_01 [{shapeType=Capsule, length=5.5, radius=4, shapeName=dyn_hood_side_r_01_capsule, center=<7 1.1 1>, rotation=<rot 0 0 -90>}]
 dyn_hood_side_r_base [{shapeType=Capsule, length=0.505, radius=0.505, shapeName=dyn_hood_side_r_base_capsule, rotation=<rot 0 -90 0>}]
 head [{shapeType=Capsule, length=8.5599, radius=2.4531, shapeName=head_capsule, rotation=<rot 0 0 90>}]
 upperarm_l [{shapeType=Capsule, length=26.4017, radius=4.6994, shapeName=upperarm_l_capsule, center=<12.8692 0 0>, rotation=<rot -90 0 0>}]
 upperarm_r [{shapeType=Capsule, length=26.4017, radius=4.6994, shapeName=upperarm_r_capsule, center=<-12.8692 0 0>, rotation=<rot -90 180 0>}]
}
constraints [
 {bone1Name=dyn_hood_side_l_01, bone2Name=dyn_hood_side_l_base, swing1LimitDegrees=15, swing2LimitDegrees=15, twistLimitDegrees=5}
 {
  bone1Name=dyn_hood_bk_l_01
  bone2Name=dyn_hood_bk_l_base
  swing1LimitDegrees=45
  swing1Motion=Locked
  swing2LimitDegrees=45
  swing2Motion=Locked
  twistLimitDegrees=45
  twistMotion=Locked
 }
 {
  bone1Name=dyn_hood_bk_r_01
  bone2Name=dyn_hood_bk_r_base
  swing1LimitDegrees=45
  swing1Motion=Locked
  swing2LimitDegrees=45
  swing2Motion=Locked
  twistLimitDegrees=45
  twistMotion=Locked
 }
 {bone1Name=dyn_hood_side_r_01, bone2Name=dyn_hood_side_r_base, swing1LimitDegrees=15, swing2LimitDegrees=15, twistLimitDegrees=5}
 {
  bone1Name=dyn_hood_bk_l_base
  bone2Name=head
  swing1LimitDegrees=45
  swing1Motion=Free
  swing2LimitDegrees=45
  swing2Motion=Free
  twistLimitDegrees=45
  twistMotion=Free
 }
 {
  bone1Name=dyn_hood_side_l_base
  bone2Name=head
  swing1LimitDegrees=45
  swing1Motion=Free
  swing2LimitDegrees=45
  swing2Motion=Free
  twistLimitDegrees=45
  twistMotion=Free
 }
 {
  bone1Name=dyn_hood_side_r_base
  bone2Name=head
  swing1LimitDegrees=45
  swing1Motion=Free
  swing2LimitDegrees=45
  swing2Motion=Free
  twistLimitDegrees=45
  twistMotion=Free
 }
 {
  bone1Name=dyn_hood_bk_r_base
  bone2Name=head
  swing1LimitDegrees=45
  swing1Motion=Free
  swing2LimitDegrees=45
  swing2Motion=Free
  twistLimitDegrees=45
  twistMotion=Free
 }
 {
  bone1Name=upperarm_l
  bone2Name=clavicle_l
  swing1LimitDegrees=45
  swing1Motion=Free
  swing2LimitDegrees=45
  swing2Motion=Free
  twistLimitDegrees=45
  twistMotion=Free
 }
 {
  bone1Name=upperarm_r
  bone2Name=clavicle_r
  swing1LimitDegrees=45
  swing1Motion=Free
  swing2LimitDegrees=45
  swing2Motion=Free
  twistLimitDegrees=45
  twistMotion=Free
 }
]
