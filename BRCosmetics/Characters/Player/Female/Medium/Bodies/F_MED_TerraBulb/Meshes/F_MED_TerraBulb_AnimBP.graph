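# Animation Blueprint for the TerraBulb female body cosmetic, adding secondary motion to its hood, coat and shoe sliders.
# Pose copied from the attached parent mesh passes through a Control Rig and hood RigidBody, then gets cached.
# Three coat RigidBody simulations switch on skydiving and DBNO states, followed by AnimDynamics on slider bones.
# Each update computes Gravity Override from parameters and sets RBAN, the inverse of bRigidBodyEnabled, for the rig.
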AnimBlueprint /BRCosmetics/Characters/Player/Female/Medium/Bodies/F_MED_TerraBulb/Meshes/F_MED_TerraBulb_AnimBP
target_skeleton=&/F_MED_TerraBulb_Skeleton
graphs [
 AnimGraph {nodes [
   n1 = RigidBody!(ComponentPose=Local To Component(LocalPose=Use cached pose 'Main Cache'()), Alpha, OverrideWorldGravity:Vector="0,0,-980", ExternalForce) -> (Pose) {
    bClampLinearTranslationLimitToRefPose=true
    bOverrideWorldGravity=true
    overridePhysicsAsset=&/F_MED_TerraBulb_Coat_NoConstraints_Physics
    simulationSpace=BaseBoneSpace
    baseBoneRef {boneName=pelvis}
    overrideWorldGravity {z=-980}
    simSpaceSettings {dampingAlpha=0, maxAngularAcceleration=40, maxAngularVelocity=8, maxLinearAcceleration=300, maxLinearVelocity=30, worldAlpha=1}
   }
   n2 = RigidBody(ComponentPose=Local To Component(LocalPose=Use cached pose 'Main Cache'()), OverrideWorldGravity="0,0,-980") {
    bClampLinearTranslationLimitToRefPose=true
    bOverrideWorldGravity=true
    overridePhysicsAsset=&/F_MED_TerraBulb_Coat_Physics
    simulationSpace=BaseBoneSpace
    baseBoneRef {boneName=pelvis}
    overrideWorldGravity {z=-980}
    simSpaceSettings {dampingAlpha=0, maxAngularAcceleration=30, maxAngularVelocity=6, maxLinearAcceleration=1300, maxLinearVelocity=400, worldAlpha=1}
   }
   n3 = Blend Poses by bool(BlendPose_0=Component To Local(ComponentPose=@n1), BlendPose_1=Component To Local(ComponentPose=@n2), bActiveValue=Get bIsDBNO())
   n4 = RigidBody(ComponentPose=Local To Component(LocalPose=Use cached pose 'Main Cache'()), OverrideWorldGravity="0,0,-980") {
    bClampLinearTranslationLimitToRefPose=true
    bOverrideWorldGravity=true
    overridePhysicsAsset=&/F_MED_TerraBulb_Coat_SkyDiving_Physics
    simulationSpace=BaseBoneSpace
    baseBoneRef {boneName=pelvis}
    overrideWorldGravity {z=-980}
    simSpaceSettings {dampingAlpha=0, maxAngularAcceleration=30, maxAngularVelocity=6, maxLinearAcceleration=300, maxLinearVelocity=40, worldAlpha=1}
   }
   n5 = Component To Local(ComponentPose=@n4)
   n6 = Get bIsSkyDiving()
   n7 = Anim Dynamics - Bone: dyn_shoes_slider_l!(
    ComponentPose=Local To Component(LocalPose=Blend Poses by bool(BlendPose_0=@n5, BlendPose_1=@n3, bActiveValue=@n6))
    Alpha=1
    GravityOverride:Vector="0,0,-980"
    ExternalForce="0,0,0"
   ) -> (Pose) {
    angularDampingOverride=0.999
    bOverrideAngularDamping=true
    bUseGravityOverride=true
    simulationSpace=RootRelative
    boundBone {boneName=dyn_shoes_slider_l}
    gravityOverride {z=-980}
    physicsBodyDefinitions [{
      boundBone {boneName=dyn_shoes_slider_l}
      boxExtents=<2 2 2>
      constraintSetup {angularLimitsMax=<0 80 15>, angularLimitsMin=<0 0 -15>, angularTarget=<0 0 0>}
      localJointOffset=<0.2 0 0.02>
     }]
   }
   n8 = Anim Dynamics - Bone: dyn_shoes_slider_r!(ComponentPose=@n7, Alpha=1, GravityOverride:Vector="0,0,-980", ExternalForce="0,0,0") -> (Pose) {
    angularDampingOverride=0.999
    bOverrideAngularDamping=true
    bUseGravityOverride=true
    simulationSpace=RootRelative
    boundBone {boneName=dyn_shoes_slider_r}
    gravityOverride {z=-980}
    physicsBodyDefinitions [{
      boundBone {boneName=dyn_shoes_slider_r}
      boxExtents=<2 2 2>
      constraintSetup {angularLimitsMax=<0 80 15>, angularLimitsMin=<0 0 -15>, angularTarget=<0 0 0>}
      localJointOffset=<-0.2 0 -0.02>
     }]
   }
   n9 = Anim Dynamics - Bone: dyn_coat_slider!(ComponentPose=@n8, Alpha=1, GravityOverride:Vector="0,0,-980", ExternalForce="0,0,0") -> (Pose) {
    angularDampingOverride=0.999
    bOverrideAngularDamping=true
    bUseGravityOverride=true
    simulationSpace=RootRelative
    boundBone {boneName=dyn_coat_slider}
    gravityOverride {z=-980}
    physicsBodyDefinitions [{
      boundBone {boneName=dyn_coat_slider}
      boxExtents=<2 2 2>
      constraintSetup {twistAxis=AxisY, angularLimitsMax=<80 10 0>, angularLimitsMin=<-2 -10 0>, angularTarget=<0 0 0>}
      localJointOffset=<0 -0.02 -0.2>
     }]
    relativeSpaceBone {boneName=dyn_coat_slider_base}
   }
   Output Pose(Result=Component To Local(ComponentPose=@n9)) {layerGroup=None}
   n10 = Control Rig!(Source=Copy Pose From Mesh() {bUseAttachedParent=true}, Alpha, DBNO:bool=Get bIsDBNO(), RBANOFF:bool=Get RBAN() -> (RBAN:bool)) -> (Pose) {controlRigAssetReference {blueprintRigClass=&/F_MED_TerraBulb_CtrlRig.F_MED_TerraBulb_CtrlRig_C}}
   n11 = RigidBody(ComponentPose=Local To Component(LocalPose=@n10), OverrideWorldGravity="0,0,-980") {
    bClampLinearTranslationLimitToRefPose=true
    bOverrideWorldGravity=true
    overridePhysicsAsset=&/F_MED_TerraBulb_Hood_Physics
    simulationSpace=BaseBoneSpace
    baseBoneRef {boneName=head}
    overrideWorldGravity {z=-980}
    simSpaceSettings {
     dampingAlpha=0
     maxAngularAcceleration=40
     maxAngularVelocity=8
     maxLinearAcceleration=300
     maxLinearVelocity=30
     velocityScaleZ=0.6
     worldAlpha=1
     externalLinearDragV=<1.5 1.5 2>
    }
   }
   Save cached pose 'Main Cache'(Pose=Component To Local(ComponentPose=@n11)) {cacheName=Main Cache}
  ]}
 EventGraph {nodes [
   Event BlueprintUpdateAnimation()
   > n1 = Sequence()
   > n2 = "Gravity Override State Function (Parameter Asset) (Nativized)"(
    CharacterPartAnimInstance=Self-Reference()
    GameTimeInSeconds=GetGameTimeInSeconds()
    Parameters=&/F_MED_TerraBulb_Parameters
   )
   > Set Gravity Override(Gravity Override=@n2)
   Set RBAN(execute=@n1.then_1, RBAN:bool=NOT Boolean(A=Get bRigidBodyEnabled())) -> (Output_Get:bool)
  ]}
]
variables [
 Gravity Override {type=Vector}
 RBAN {type=boolean}
]
