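# Physics asset of capsule colliders for the TerraBulb medium female player body, named for use with its cape.
# Capsules cover the pelvis, spine, neck, head, arms and legs, plus the dyn_coat front, side and back-left bone chains; there is no back-right chain.
# Constraints follow the limb hierarchy, with each coat chain rooted at spine_01.
# All swing and twist motions are Free, so the constraints impose no angular limits.
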PhysicsAsset /BRCosmetics/Characters/Player/Female/Medium/Bodies/F_MED_TerraBulb/Meshes/F_MED_TerraBulb_CapeColliders
bodies {
 calf_l [{shapeType=Capsule, length=40.8905, radius=11.6472, shapeName=calf_l_capsule, center=<-24.3972 -2.1288 0>, rotation=<rot 90 180 0>}]
 calf_r [{shapeType=Capsule, length=40.8905, radius=11.6472, shapeName=calf_r_capsule, center=<24.3972 2.1288 0>, rotation=<rot 90 0 0>}]
 dyn_coat_bk_l_01 [{shapeType=Capsule, length=10, radius=8, shapeName=dyn_coat_bk_l_01_capsule, center=<5.1218 0 1.5122>, rotation=<rot 0 0 90>}]
 dyn_coat_bk_l_02 [{shapeType=Capsule, length=10, radius=8, shapeName=dyn_coat_bk_l_02_capsule, center=<3.6204 0 1.5122>, rotation=<rot 0 0 90>}]
 dyn_coat_bk_l_03 [{shapeType=Capsule, length=10, radius=8, shapeName=dyn_coat_bk_l_03_capsule, center=<3.6204 0 1.5122>, rotation=<rot 0 0 90>}]
 dyn_coat_fr_l_01 [{shapeType=Capsule, length=6, radius=7, shapeName=dyn_coat_fr_l_01_capsule, center=<8.8271 0 1.3591>, rotation=<rot 0 0 90>}]
 dyn_coat_fr_l_02 [{shapeType=Capsule, length=10, radius=6, shapeName=dyn_coat_fr_l_02_capsule, center=<6.253 0 1.1409>, rotation=<rot 0 0 90>}]
 dyn_coat_fr_l_03 [{shapeType=Capsule, length=10, radius=7, shapeName=dyn_coat_fr_l_03_capsule, center=<5.3343 0 1.3544>, rotation=<rot 0 0 90>}]
 dyn_coat_fr_l_04 [{
   shapeType=Capsule
   length=11
   radius=8
   shapeName=dyn_coat_fr_l_04_capsule
   center=<7.5647 -0.1553 1.5357>
   rotation=<rot 2.0869 21.1184 84.6139>
  }]
 dyn_coat_fr_r_01 [{shapeType=Capsule, length=6, radius=7, shapeName=dyn_coat_fr_r_01_capsule, center=<8.8271 0 1.3591>, rotation=<rot 0 0 -90>}]
 dyn_coat_fr_r_02 [{shapeType=Capsule, length=10, radius=6, shapeName=dyn_coat_fr_r_02_capsule, center=<6.253 0 1.1409>, rotation=<rot 0 0 -90>}]
 dyn_coat_fr_r_03 [{shapeType=Capsule, length=10, radius=7, shapeName=dyn_coat_fr_r_03_capsule, center=<5.3343 0 1.3544>, rotation=<rot 0 0 -90>}]
 dyn_coat_fr_r_04 [{
   shapeType=Capsule
   length=11
   radius=8
   shapeName=dyn_coat_fr_r_04_capsule
   center=<7.5647 0.1553 1.5357>
   rotation=<rot -2.0869 -21.1184 -95.386>
  }]
 dyn_coat_side_l_01 [{shapeType=Capsule, length=10, radius=7, shapeName=dyn_coat_side_l_01_capsule, center=<6.0106 0 1.7679>, rotation=<rot 0 0 87.3793>}]
 dyn_coat_side_l_02 [{shapeType=Capsule, length=10, radius=7, shapeName=dyn_coat_side_l_02_capsule, center=<5.03 0 1.7679>, rotation=<rot 0 0 87.3793>}]
 dyn_coat_side_l_03 [{shapeType=Capsule, length=10, radius=7, shapeName=dyn_coat_side_l_03_capsule, center=<3.2783 0 1.7679>, rotation=<rot 0 22.6775 87.3793>}]
 dyn_coat_side_r_01 [{shapeType=Capsule, length=10, radius=7, shapeName=dyn_coat_side_r_01_capsule, center=<6.0106 0 1.7679>, rotation=<rot 0 0 -91.3997>}]
 dyn_coat_side_r_02 [{shapeType=Capsule, length=10, radius=7, shapeName=dyn_coat_side_r_02_capsule, center=<5.03 0 1.7679>, rotation=<rot 0 0 -91.3997>}]
 dyn_coat_side_r_03 [{
   shapeType=Capsule
   length=10
   radius=7
   shapeName=dyn_coat_side_r_03_capsule
   center=<3.2783 0 1.7679>
   rotation=<rot 0.0001 -18.135 -91.3997>
  }]
 foot_l [{shapeType=Capsule, length=11.4142, radius=10.2543, shapeName=foot_l_capsule, center=<-1.7864 6.3626 0>, rotation=<rot 0 0 -90>}]
 foot_r [{shapeType=Capsule, length=11.4142, radius=10.2543, shapeName=foot_r_capsule, center=<1.7864 -6.3626 0>, rotation=<rot 0 0 90>}]
 hand_l [{shapeType=Capsule, length=4.2506, radius=8.1487, shapeName=hand_l_capsule, center=<7.9367 -0.4376 -0.1853>, rotation=<rot 90 172.2236 0>}]
 hand_r [{shapeType=Capsule, length=4.2506, radius=8.1487, shapeName=hand_r_capsule, center=<-7.9367 0.4376 0.1853>, rotation=<rot 90 -7.7764 0>}]
 head [{shapeType=Capsule, length=3.686, radius=12.942, shapeName=head_capsule}]
 lowerarm_l [{shapeType=Capsule, length=23.005, radius=5.005, shapeName=lowerarm_l_capsule, center=<10.6656 0 0>, rotation=<rot 90 180 0>}]
 lowerarm_r [{shapeType=Capsule, length=23.005, radius=5.005, shapeName=lowerarm_r_capsule, center=<-10.6656 0 0>, rotation=<rot 90 0 0>}]
 neck_01 [{shapeType=Capsule, length=2.505, radius=7.505, shapeName=neck_01_capsule, center=<0 0.7507 0>}]
 neck_02 [{shapeType=Capsule, length=7.2501, radius=5.4717, shapeName=neck_02_capsule}]
 pelvis [{shapeType=Capsule, length=9.0261, radius=12.3102, shapeName=pelvis_capsule}]
 spine_01 [{shapeType=Capsule, length=10.505, radius=10.005, shapeName=spine_01_capsule}]
 spine_02 [{shapeType=Capsule, length=11.2373, radius=9.0626, shapeName=spine_02_capsule}]
 spine_03 [{shapeType=Capsule, length=8.51, radius=9, shapeName=spine_03_capsule, center=<0 2.0065 0>}]
 spine_04 [{shapeType=Capsule, length=4.9173, radius=10.6376, shapeName=spine_04_capsule, center=<0 1.5704 0>}]
 spine_05 [{shapeType=Capsule, length=12.0284, radius=11.5064, shapeName=spine_05_capsule, center=<0 0.6191 0>}]
 thigh_l [{shapeType=Capsule, length=39, radius=10.6488, shapeName=thigh_l_capsule, center=<-24.4134 -0.3514 0>, rotation=<rot 90 -175.9599 0>}]
 thigh_r [{shapeType=Capsule, length=39, radius=10.6488, shapeName=thigh_r_capsule, center=<24.4134 0.3514 0>, rotation=<rot 90 4.0401 0>}]
 upperarm_l [{shapeType=Capsule, length=24.924, radius=5.3312, shapeName=upperarm_l_capsule, center=<13.3437 0 0>, rotation=<rot 90 180 0>}]
 upperarm_r [{shapeType=Capsule, length=24.924, radius=5.3312, shapeName=upperarm_r_capsule, center=<-13.3437 0 0>, rotation=<rot 90 0 0>}]
}
constraints [
 {
  bone1Name=spine_01
  bone2Name=pelvis
  swing1LimitDegrees=45
  swing1Motion=Free
  swing2LimitDegrees=45
  swing2Motion=Free
  twistLimitDegrees=45
  twistMotion=Free
 }
 {
  bone1Name=spine_02
  bone2Name=spine_01
  swing1LimitDegrees=45
  swing1Motion=Free
  swing2LimitDegrees=45
  swing2Motion=Free
  twistLimitDegrees=45
  twistMotion=Free
 }
 {
  bone1Name=spine_03
  bone2Name=spine_02
  swing1LimitDegrees=45
  swing1Motion=Free
  swing2LimitDegrees=45
  swing2Motion=Free
  twistLimitDegrees=45
  twistMotion=Free
 }
 {
  bone1Name=spine_04
  bone2Name=spine_03
  swing1LimitDegrees=45
  swing1Motion=Free
  swing2LimitDegrees=45
  swing2Motion=Free
  twistLimitDegrees=45
  twistMotion=Free
 }
 {
  bone1Name=spine_05
  bone2Name=spine_04
  swing1LimitDegrees=45
  swing1Motion=Free
  swing2LimitDegrees=45
  swing2Motion=Free
  twistLimitDegrees=45
  twistMotion=Free
 }
 {
  bone1Name=hand_l
  bone2Name=lowerarm_l
  swing1LimitDegrees=45
  swing1Motion=Free
  swing2LimitDegrees=45
  swing2Motion=Free
  twistLimitDegrees=45
  twistMotion=Free
 }
 {
  bone1Name=upperarm_r
  bone2Name=spine_05
  swing1LimitDegrees=45
  swing1Motion=Free
  swing2LimitDegrees=45
  swing2Motion=Free
  twistLimitDegrees=45
  twistMotion=Free
 }
 {
  bone1Name=upperarm_l
  bone2Name=spine_05
  swing1LimitDegrees=45
  swing1Motion=Free
  swing2LimitDegrees=45
  swing2Motion=Free
  twistLimitDegrees=45
  twistMotion=Free
 }
 {
  bone1Name=lowerarm_l
  bone2Name=upperarm_l
  swing1LimitDegrees=45
  swing1Motion=Free
  swing2LimitDegrees=45
  swing2Motion=Free
  twistLimitDegrees=45
  twistMotion=Free
 }
 {
  bone1Name=lowerarm_r
  bone2Name=upperarm_r
  swing1LimitDegrees=45
  swing1Motion=Free
  swing2LimitDegrees=45
  swing2Motion=Free
  twistLimitDegrees=45
  twistMotion=Free
 }
 {
  bone1Name=hand_r
  bone2Name=lowerarm_r
  swing1LimitDegrees=45
  swing1Motion=Free
  swing2LimitDegrees=45
  swing2Motion=Free
  twistLimitDegrees=45
  twistMotion=Free
 }
 {
  bone1Name=head
  bone2Name=neck_02
  swing1LimitDegrees=45
  swing1Motion=Free
  swing2LimitDegrees=45
  swing2Motion=Free
  twistLimitDegrees=45
  twistMotion=Free
 }
 {
  bone1Name=neck_01
  bone2Name=spine_05
  swing1LimitDegrees=45
  swing1Motion=Free
  swing2LimitDegrees=45
  swing2Motion=Free
  twistLimitDegrees=45
  twistMotion=Free
 }
 {
  bone1Name=neck_02
  bone2Name=neck_01
  swing1LimitDegrees=45
  swing1Motion=Free
  swing2LimitDegrees=45
  swing2Motion=Free
  twistLimitDegrees=45
  twistMotion=Free
 }
 {
  bone1Name=thigh_l
  bone2Name=pelvis
  swing1LimitDegrees=45
  swing1Motion=Free
  swing2LimitDegrees=45
  swing2Motion=Free
  twistLimitDegrees=45
  twistMotion=Free
 }
 {
  bone1Name=calf_l
  bone2Name=thigh_l
  swing1LimitDegrees=45
  swing1Motion=Free
  swing2LimitDegrees=45
  swing2Motion=Free
  twistLimitDegrees=45
  twistMotion=Free
 }
 {
  bone1Name=foot_l
  bone2Name=calf_l
  swing1LimitDegrees=45
  swing1Motion=Free
  swing2LimitDegrees=45
  swing2Motion=Free
  twistLimitDegrees=45
  twistMotion=Free
 }
 {
  bone1Name=thigh_r
  bone2Name=pelvis
  swing1LimitDegrees=45
  swing1Motion=Free
  swing2LimitDegrees=45
  swing2Motion=Free
  twistLimitDegrees=45
  twistMotion=Free
 }
 {
  bone1Name=calf_r
  bone2Name=thigh_r
  swing1LimitDegrees=45
  swing1Motion=Free
  swing2LimitDegrees=45
  swing2Motion=Free
  twistLimitDegrees=45
  twistMotion=Free
 }
 {
  bone1Name=foot_r
  bone2Name=calf_r
  swing1LimitDegrees=45
  swing1Motion=Free
  swing2LimitDegrees=45
  swing2Motion=Free
  twistLimitDegrees=45
  twistMotion=Free
 }
 {
  bone1Name=dyn_coat_bk_l_01
  bone2Name=spine_01
  swing1LimitDegrees=45
  swing1Motion=Free
  swing2LimitDegrees=45
  swing2Motion=Free
  twistLimitDegrees=45
  twistMotion=Free
 }
 {
  bone1Name=dyn_coat_bk_l_02
  bone2Name=dyn_coat_bk_l_01
  swing1LimitDegrees=45
  swing1Motion=Free
  swing2LimitDegrees=45
  swing2Motion=Free
  twistLimitDegrees=45
  twistMotion=Free
 }
 {
  bone1Name=dyn_coat_bk_l_03
  bone2Name=dyn_coat_bk_l_02
  swing1LimitDegrees=45
  swing1Motion=Free
  swing2LimitDegrees=45
  swing2Motion=Free
  twistLimitDegrees=45
  twistMotion=Free
 }
 {
  bone1Name=dyn_coat_fr_l_01
  bone2Name=spine_01
  swing1LimitDegrees=45
  swing1Motion=Free
  swing2LimitDegrees=45
  swing2Motion=Free
  twistLimitDegrees=45
  twistMotion=Free
 }
 {
  bone1Name=dyn_coat_fr_l_02
  bone2Name=dyn_coat_fr_l_01
  swing1LimitDegrees=45
  swing1Motion=Free
  swing2LimitDegrees=45
  swing2Motion=Free
  twistLimitDegrees=45
  twistMotion=Free
 }
 {
  bone1Name=dyn_coat_fr_l_03
  bone2Name=dyn_coat_fr_l_02
  swing1LimitDegrees=45
  swing1Motion=Free
  swing2LimitDegrees=45
  swing2Motion=Free
  twistLimitDegrees=45
  twistMotion=Free
 }
 {
  bone1Name=dyn_coat_fr_l_04
  bone2Name=dyn_coat_fr_l_03
  swing1LimitDegrees=45
  swing1Motion=Free
  swing2LimitDegrees=45
  swing2Motion=Free
  twistLimitDegrees=45
  twistMotion=Free
 }
 {
  bone1Name=dyn_coat_fr_r_01
  bone2Name=spine_01
  swing1LimitDegrees=45
  swing1Motion=Free
  swing2LimitDegrees=45
  swing2Motion=Free
  twistLimitDegrees=45
  twistMotion=Free
 }
 {
  bone1Name=dyn_coat_fr_r_02
  bone2Name=dyn_coat_fr_r_01
  swing1LimitDegrees=45
  swing1Motion=Free
  swing2LimitDegrees=45
  swing2Motion=Free
  twistLimitDegrees=45
  twistMotion=Free
 }
 {
  bone1Name=dyn_coat_fr_r_03
  bone2Name=dyn_coat_fr_r_02
  swing1LimitDegrees=45
  swing1Motion=Free
  swing2LimitDegrees=45
  swing2Motion=Free
  twistLimitDegrees=45
  twistMotion=Free
 }
 {
  bone1Name=dyn_coat_fr_r_04
  bone2Name=dyn_coat_fr_r_03
  swing1LimitDegrees=45
  swing1Motion=Free
  swing2LimitDegrees=45
  swing2Motion=Free
  twistLimitDegrees=45
  twistMotion=Free
 }
 {
  bone1Name=dyn_coat_side_l_01
  bone2Name=spine_01
  swing1LimitDegrees=45
  swing1Motion=Free
  swing2LimitDegrees=45
  swing2Motion=Free
  twistLimitDegrees=45
  twistMotion=Free
 }
 {
  bone1Name=dyn_coat_side_l_02
  bone2Name=dyn_coat_side_l_01
  swing1LimitDegrees=45
  swing1Motion=Free
  swing2LimitDegrees=45
  swing2Motion=Free
  twistLimitDegrees=45
  twistMotion=Free
 }
 {
  bone1Name=dyn_coat_side_l_03
  bone2Name=dyn_coat_side_l_02
  swing1LimitDegrees=45
  swing1Motion=Free
  swing2LimitDegrees=45
  swing2Motion=Free
  twistLimitDegrees=45
  twistMotion=Free
 }
 {
  bone1Name=dyn_coat_side_r_01
  bone2Name=spine_01
  swing1LimitDegrees=45
  swing1Motion=Free
  swing2LimitDegrees=45
  swing2Motion=Free
  twistLimitDegrees=45
  twistMotion=Free
 }
 {
  bone1Name=dyn_coat_side_r_02
  bone2Name=dyn_coat_side_r_01
  swing1LimitDegrees=45
  swing1Motion=Free
  swing2LimitDegrees=45
  swing2Motion=Free
  twistLimitDegrees=45
  twistMotion=Free
 }
 {
  bone1Name=dyn_coat_side_r_03
  bone2Name=dyn_coat_side_r_02
  swing1LimitDegrees=45
  swing1Motion=Free
  swing2LimitDegrees=45
  swing2Motion=Free
  twistLimitDegrees=45
  twistMotion=Free
 }
]
