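# Animation Blueprint for the TerraBulb female face accessory mesh, adding secondary physics motion to its hair bones.
# The AnimGraph copies the Body mesh pose, chains eight angle-limited, heavily damped Anim Dynamics nodes, then offsets and squashes the Correction_hair bone.
# Body is set on initialization or customization load, using the custom base mesh if valid, otherwise the owning component.
# Lod_Fix updates from the current LOD, but the AnimGraph never reads it.
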AnimBlueprint /BRCosmetics/Characters/Player/Female/Medium/Bodies/F_MED_TerraBulb/Meshes/Parts/F_MED_TerraBulb_FaceAcc_AnimBP
target_skeleton=&/F_MED_TerraBulb_FaceAcc_Skeleton
graphs [
 AnimGraph {nodes [
   n1 = Get Body() -> (Body:Skeletal Mesh Component Object Reference)
   n2 = Anim Dynamics - Bone: dyn_hair_1_1(
    ComponentPose=Local To Component(LocalPose=Copy Pose From Mesh(SourceMeshComponent=@n1) {bUseAttachedParent=true, bUseMeshPose=true})
    Alpha=1
    ExternalForce="0,0,0"
   ) {
    angularDampingOverride=0.999
    bOverrideAngularDamping=true
    bUseGravityOverride=true
    simulationSpace=RootRelative
    boundBone {boneName=dyn_hair_1_1}
    gravityOverride {z=-980}
    physicsBodyDefinitions [{
      boundBone {boneName=dyn_hair_1_1}
      boxExtents=<4.5 4.5 4.5>
      constraintSetup {angularLimitsMax=<0 30 10>, angularLimitsMin=<0 0 -10>, angularTarget=<0 0 0>}
      localJointOffset=<0.2 0 0>
     }]
   }
   n3 = Anim Dynamics - Bone: dyn_hair_2_1(ComponentPose=@n2, Alpha=1, ExternalForce="0,0,0") {
    angularDampingOverride=0.999
    bOverrideAngularDamping=true
    bUseGravityOverride=true
    simulationSpace=RootRelative
    boundBone {boneName=dyn_hair_2_1}
    gravityOverride {z=-980}
    physicsBodyDefinitions [{
      boundBone {boneName=dyn_hair_2_1}
      boxExtents=<5 5 5>
      constraintSetup {angularLimitsMax=<0 30 10>, angularLimitsMin=<0 0 -10>, angularTarget=<0 0 0>}
      localJointOffset=<0.2 0 0>
     }]
   }
   n4 = Anim Dynamics - Bone: dyn_hair_3_1(ComponentPose=@n3, Alpha=1, ExternalForce="0,0,0") {
    angularDampingOverride=0.999
    bOverrideAngularDamping=true
    bUseGravityOverride=true
    simulationSpace=RootRelative
    boundBone {boneName=dyn_hair_3_1}
    gravityOverride {z=-980}
    physicsBodyDefinitions [{
      boundBone {boneName=dyn_hair_3_1}
      boxExtents=<6 6 6>
      constraintSetup {angularLimitsMax=<0 15 15>, angularLimitsMin=<0 0 -15>, angularTarget=<0 0 0>}
      localJointOffset=<0.2 0 0>
     }]
   }
   n5 = Anim Dynamics - Bone: dyn_hair_4_1(ComponentPose=@n4, Alpha=1, ExternalForce="0,0,0") {
    angularDampingOverride=0.999
    bOverrideAngularDamping=true
    bUseGravityOverride=true
    simulationSpace=RootRelative
    boundBone {boneName=dyn_hair_4_1}
    gravityOverride {z=-980}
    physicsBodyDefinitions [{
      boundBone {boneName=dyn_hair_4_1}
      boxExtents=<7 7 7>
      constraintSetup {angularLimitsMax=<0 15 15>, angularLimitsMin=<0 0 -15>, angularTarget=<0 0 0>}
      localJointOffset=<0.2 0 0>
     }]
   }
   n6 = Anim Dynamics - Bone: dyn_hair_5_1(ComponentPose=@n5, Alpha=1, ExternalForce="0,0,0") {
    angularDampingOverride=0.999
    bOverrideAngularDamping=true
    bUseGravityOverride=true
    simulationSpace=RootRelative
    boundBone {boneName=dyn_hair_5_1}
    gravityOverride {z=-980}
    physicsBodyDefinitions [{
      boundBone {boneName=dyn_hair_5_1}
      boxExtents=<5.5 5.5 5.5>
      constraintSetup {angularLimitsMax=<0 15 15>, angularLimitsMin=<0 0 -20>, angularTarget=<0 0 0>}
      localJointOffset=<0.2 0 0>
     }]
   }
   n7 = Anim Dynamics - Bone: dyn_hair_6_1(ComponentPose=@n6, Alpha=1, ExternalForce="0,0,0") {
    angularDampingOverride=0.999
    bOverrideAngularDamping=true
    bUseGravityOverride=true
    simulationSpace=RootRelative
    boundBone {boneName=dyn_hair_6_1}
    gravityOverride {z=-980}
    physicsBodyDefinitions [{
      boundBone {boneName=dyn_hair_6_1}
      boxExtents=<6 6 6>
      constraintSetup {angularLimitsMax=<0 15 15>, angularLimitsMin=<0 0 -20>, angularTarget=<0 0 0>}
      localJointOffset=<0.2 0 0>
     }]
   }
   n8 = Anim Dynamics - Bone: dyn_hair_7_1(ComponentPose=@n7, Alpha=1, ExternalForce="0,0,0") {
    angularDampingOverride=0.999
    bOverrideAngularDamping=true
    bUseGravityOverride=true
    simulationSpace=RootRelative
    boundBone {boneName=dyn_hair_7_1}
    gravityOverride {z=-980}
    physicsBodyDefinitions [{
      boundBone {boneName=dyn_hair_7_1}
      boxExtents=<6 6 6>
      constraintSetup {angularLimitsMax=<0 15 15>, angularLimitsMin=<0 0 -20>, angularTarget=<0 0 0>}
      localJointOffset=<0.2 0 0>
     }]
   }
   n9 = Anim Dynamics - Bone: dyn_hair_8_1(ComponentPose=@n8, Alpha=1, ExternalForce="0,0,0") {
    angularDampingOverride=0.999
    bOverrideAngularDamping=true
    bUseGravityOverride=true
    simulationSpace=RootRelative
    boundBone {boneName=dyn_hair_8_1}
    gravityOverride {z=-980}
    physicsBodyDefinitions [{
      boundBone {boneName=dyn_hair_8_1}
      boxExtents=<6 6 6>
      constraintSetup {angularLimitsMax=<0 15 15>, angularLimitsMin=<0 0 -20>, angularTarget=<0 0 0>}
      localJointOffset=<0.2 0 0>
     }]
   }
   n10 = "Transform (Modify) Bone - Bone: Correction_hair"(ComponentPose=@n9, Alpha=1, Translation="1,0,0", Rotation="0,0,0", Scale="1,1,0.8") {
    scaleMode=BMM_Replace
    scaleSpace=BCS_BoneSpace
    translationMode=BMM_Replace
    translationSpace=BCS_BoneSpace
    boneToModify {boneName=Correction_hair}
    scale {z=0.8}
    translation {x=1}
   }
   Output Pose(Result=Component To Local(ComponentPose=@n10)) {layerGroup=None}
  ]}
 EventGraph {nodes [
   Event BlueprintUpdateAnimation()
   > Set Lod_Fix(Lod_Fix:bool="integer >= integer"(A=Get CurrentLOD(), B=1)) -> (Output_Get:bool)
   n1 = Event OnCharacterCustomizationLoaded()
   n2 = Is Valid(exec=[@n1.then, @n3.then], InputObject=@n1.CustomBaseMesh)
   Set Body(execute=@n2.Is Not Valid, Body:Skeletal Mesh Component Object Reference=GetOwnerSkeletalMeshComponent(self=Self-Reference())) -> (Output_Get:Skeletal Mesh Component Object Reference)
   Set Body(execute=@n2.Is Valid, Body=@n1.CustomBaseMesh)
   n3 = Event BlueprintInitializeAnimation()
  ]}
]
variables [
 Body {type=SkeletalMeshComponent}
 Lod_Fix {type=boolean}
]
